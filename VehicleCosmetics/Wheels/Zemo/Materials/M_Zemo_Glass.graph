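# Translucent lit glass material for the Zemo vehicle wheel, built on material attributes for skeletal meshes.
# Mask channels, inverted Fresnel and a custom color build ring and tint highlights, darkened by the green channel; blue scales opacity.
# These "emissive" terms feed Base Color, as Emissive is never set.
# Color, opacity and normal map pass through shared vehicle damage, then de-rez transition functions.
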Material /VehicleCosmetics/Wheels/Zemo/Materials/M_Zemo_Glass
graph {
 properties {
  bAutomaticallySetUsageInEditor=false
  bUseMaterialAttributes=true
  bUsedWithMeshDeformer=true
  bUsedWithSkeletalMesh=true
  blendMode=BLEND_Translucent
  previewMesh=/VehicleCosmetics/Wheels/Zemo/Meshes/SK_Wheel_Zemo
  usedShadingModels=Default Lit
 }
 nodes [
  n1 = NamedRerouteUsage()
  n2 = NamedRerouteUsage()
  n3 = TextureSampleParameter2D() {parameterName=Normal, samplerType=SAMPLERTYPE_Normal, texture=/VehicleCosmetics/Wheels/Zemo/Textures/T_Zemo_N}
  MaterialFunctionCall(In=MaterialFunctionCall(Material=SetMaterialAttributes!(MaterialAttributes, Base Color=@n1, Opacity=@n2, Normal=@n3.RGB) {attributeSetTypes [#1, #2, #3]}) {materialFunction=&MF_BaseVehicle_Damage}.Result) {materialFunction=&MF_VehicleDeRez_Transition}
  n4 = TextureSampleParameter2D() {parameterName=Mask, samplerType=SAMPLERTYPE_LinearColor, texture=/VehicleCosmetics/Wheels/Zemo/Textures/T_Zemo_M}
  NamedRerouteDeclaration(None=Multiply(A=@n4.B, B=ScalarParameter() {defaultValue=0.5, group=00 Base, parameterName=Opacity})) {name=Opacity, nodeColor=<rgba 1 0.1406 0 1>}
  n5 = ScalarParameter() {defaultValue=10, group=00 Base, parameterName=Ring Emissive Power}
  n6 = Multiply(A=OneMinus(None=Fresnel() {exponent=3}), B=ScalarParameter() {defaultValue=0.4, group=00 Base, parameterName=Fresnel Amount})
  n7 = VectorParameter() {parameterName=Custom Color, defaultValue=<rgba 0.1912 0.6795 0.9387 1>}
  n8 = ScalarParameter() {defaultValue=0.3, group=00 Base, parameterName=Emissive Amount}
  n9 = Add(A=Multiply(A=Multiply(A=@n4.R, B=@n5), B=@n6), B=Add(A=@n6, B=Multiply(A=@n7.RGB, B=@n8)))
  n10 = Multiply(
   A=Multiply(A=@n4.A, B=ScalarParameter() {defaultValue=0.07, group=00 Base, parameterName=Coin Ring Intensity})
   B=Multiply(A=@n7.RGB) {constB=5}
  )
  NamedRerouteDeclaration(None=LinearInterpolate(A=Add(A=@n9, B=@n10), Alpha=OneMinus(None=@n4.G)) {constB=0.01}) {name=BC, nodeColor=<rgba 0 0.1797 1 1>}
 ]
}
