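# Translucent, two-sided glass material for the Zemo vehicle wheel cosmetic, usable on skeletal meshes.
# Base color combines a tintable color, Fresnel and mask-driven ring highlights, darkened to near black where a second mask's green channel is low.
# Fresnel also drives tinted emissive, while the main mask's blue channel scales both roughness and opacity.
# The material attributes then pass through shared vehicle damage and de-rez transition functions.
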Material /VehicleCosmetics/Wheels/Zemo/Materials/M_Zemo_Glass_02
graph {
 properties {
  bAutomaticallySetUsageInEditor=false
  bUseMaterialAttributes=true
  bUsedWithSkeletalMesh=true
  blendMode=BLEND_Translucent
  translucencyLightingMode=TLM_Surface
  twoSided=true
  usedShadingModels=Default Lit
 }
 nodes [
  n1 = ScalarParameter() {defaultValue=10, group=00 Base, parameterName=Ring Emissive Power}
  n2 = TextureSampleParameter2D() {parameterName=Mask, samplerType=SAMPLERTYPE_LinearColor, texture=/VehicleCosmetics/Wheels/Zemo/Textures/T_Zemo_M}
  n3 = Multiply(
   A=Fresnel(ExponentIn=ScalarParameter() {defaultValue=3, parameterName=Fresnel_exp}) {baseReflectFraction=0.02, exponent=3}
   B=ScalarParameter() {defaultValue=0.3, group=00 Base, parameterName=Fresnel Amount}
  )
  n4 = VectorParameter() {parameterName=Custom Color, defaultValue=<rgba 0.2189 0.9387 0.2634 1>}
  n5 = ScalarParameter() {defaultValue=0.5, group=00 Base, parameterName=Emissive Amount}
  n6 = Add(A=Multiply(A=Multiply(A=@n2.R, B=@n1), B=@n3), B=Add(A=@n3, B=Multiply(A=@n4.RGB, B=@n5)))
  n7 = Multiply(
   A=Multiply(A=@n2.A, B=ScalarParameter() {defaultValue=0.07, group=00 Base, parameterName=Coin Ring Intensity})
   B=Multiply(A=@n4.RGB) {constB=5}
  )
  n8 = TextureSampleParameter2D() {parameterName=Mask2, texture=/VehicleCosmetics/Wheels/Zemo/Textures/T_Zemo_RGB}
  n9 = SetMaterialAttributes!(
   MaterialAttributes
   Base Color=LinearInterpolate(A=Add(A=@n6, B=@n7), Alpha=OneMinus(None=@n8.G)) {constB=0.01}
   Emissive Color=NamedRerouteUsage()
   Normal=TextureSampleParameter2D() {parameterName=Normal, samplerType=SAMPLERTYPE_Normal, texture=/VehicleCosmetics/Wheels/Zemo/Textures/T_Zemo_N}.RGB
   Metallic=ScalarParameter() {defaultValue=0.3, parameterName=Metalic}
   Roughness=Multiply(A=@n2.B, B=ScalarParameter() {defaultValue=0.15, parameterName=Roughness})
   Specular=ScalarParameter() {defaultValue=0.4, parameterName=Sprcular}
   Opacity=NamedRerouteUsage()
  ) {attributeSetTypes [
    #1
    #2
    #3
    #4
    #5
    #6
    #7
   ]}
  MaterialFunctionCall(In=MaterialFunctionCall(Material=@n9) {materialFunction=&MF_BaseVehicle_Damage}.Result) {materialFunction=&MF_VehicleDeRez_Transition}
  NamedRerouteDeclaration(None=Multiply(A=@n2.B, B=ScalarParameter() {defaultValue=1, group=00 Base, parameterName=Opacity})) {name=Opacity, nodeColor=<rgba 1 0.1406 0 1>}
  n10 = ScalarParameter() {defaultValue=2, parameterName=Em int}
  NamedRerouteDeclaration(None=Multiply(A=Multiply(A=@n10, B=@n3), B=VectorParameter() {parameterName=Emissive Color, defaultValue=<rgba 0.2189 0.9387 0.2634 1>}.RGB)) {name=em, nodeColor=<rgba 1 0.0938 0 1>}
  Reroute(None=@n2.G)
 ]
}
